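# Niagara fire effect for the baking part of the SoapPocket emote, combining flame meshes, fire sprites and a glowing flare.
# Two delayed mesh bursts vary size, lifetime and rotation by particle partition, alongside continuous and burst-spawned rising sprites.
# A persistent flare drives sprites and a light, fading in and later flickering by system age.
# System.ScaleUpCurve scales sprites and light radius; the rope user parameters are unreferenced in the emitters shown.
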NiagaraSystem /BRCosmetics/Effects/Fort_Effects/Emotes/SoapPocket/NS_NS_SoapPocket_Baking
dependencies {
 usedDataInterfaces [
  &NiagaraDataInterfaceCurve
  &NiagaraDataInterfaceColorCurve
  &NiagaraDataInterfaceVectorCurve
  &NiagaraDataInterfaceArrayFloat
 ]
 usedDynamicInputs [
  &VectorFromCurve
  &SimulationPosition
  &RemapRange
  &FN_LOD_Distance_Remap
 ]
 usedModules [
  &self:SystemSpawnScript.SystemScriptSource.SystemScriptGraph.NiagaraNodeAssignment_0.SetVariables_#1
  &SystemState
  &self:SystemSpawnScript.SystemScriptSource.SystemScriptGraph.NiagaraNodeAssignment_2.SetVariables_#2
  &self:SystemSpawnScript.SystemScriptSource.SystemScriptGraph.NiagaraNodeAssignment_1.SetVariables_#3
 ]
}
emitters {
 FireTrailMesh {
  emitterUpdateScript [
   &EmitterState(Inactive Response=::NewEnumerator1, Life Cycle Mode=::NewEnumerator1, Loop Behavior=::NewEnumerator1, Loop Duration=1, UseLoopDelay=false)
   &SpawnBurst_Instantaneous(Spawn Count=2, Spawn Time=0.4, Use Spawn Probability=false)
  ]
  particleSpawnScript [
   &PartitionParticles(Number Of Partitions=4)
   &InitializeParticle(
    Lifetime=&Multiply_Float(
     A=0.7
     B=&SelectFloatFromArray(
      Direct Array Index=$Particles.PartitionParticles.Partition
      Float Selection Array=DataInterface() {propertyValues {FloatData [1,0.9,0.8,0.7]}}
     )
    )
    Mesh Scale Mode=::NewEnumerator3
    Mesh Uniform Scale=&Multiply_Float(
     A=&SelectFloatFromArray(
      Direct Array Index=$Particles.PartitionParticles.Partition
      Float Selection Array=DataInterface() {propertyValues {FloatData [1,0.68,0.45,0.26]}}
     )
     B=0.45
    )
    UsePositionOffset=false
    Write Material Random=true
   ) {module=InitializeParticle001}
   &InitialMeshOrientation(
    Mesh Orientation Mode=::NewEnumerator2
    Rotation=&MakeVector(
     X=&SelectFloatFromArray(
      Direct Array Index=$Particles.PartitionParticles.Partition
      Float Selection Array=DataInterface() {propertyValues {FloatData [0,-0.03,-0.07,-0.09]}}
     )
     Y=0
     Z=0
    )
    Use Rotation Vector=true
   )
   &SystemLocation(MaskSpawnGroup=false, Offset=<0 0 10>)
  ]
  particleUpdateScript [
   &ParticleState()
   &ScaleMeshSize(Scale Factor=<1 1.5 1>)
   &SolveForcesAndVelocity(Clamp Velocity=false, Limit Acceleration=false)
   &DynamicMaterialParameters(
    Index 0 Param 1=&FloatFromCurve(
     FloatCurve=DataInterface() {propertyValues {Curve {keys [{arriveTangent=3.1068, interpMode=RCIM_Cubic, leaveTangent=3.1068, tangentMode=RCTM_User, time=0, value=0}, <at 1 1 RCIM_Cubic>]}}}
     Scale Curve=0.7
    )
    Index 0 Param 2=&Multiply_Float(
     A=1
     B=&SelectFloatFromArray(Direct Array Index=$Particles.PartitionParticles.Partition, Float Selection Array=DataInterface() {propertyValues {FloatData [1,1,1,1]}})
    )
    Index 0 Param 3=1
    Index 0 Param 4=2
    Module.Index0Float0Write=true
    Module.Index0Float1Write=true
    Module.Index0Float2Write=true
    Module.Index0Float3Write=true
   )
   &ScaleColor(
    Scale Alpha=&FloatFromCurve(FloatCurve=DataInterface() {propertyValues {Curve {keys [
         {interpMode=RCIM_Cubic, tangentMode=RCTM_User, time=0, value=0}
         {arriveTangent=0.25, interpMode=RCIM_Cubic, leaveTangent=0.25, time=0.3, value=1}
         {arriveTangent=-1.4286, interpMode=RCIM_Cubic, leaveTangent=-1.4286, time=0.6, value=0.15}
         {arriveTangent=-0.0554, interpMode=RCIM_Cubic, leaveTangent=-0.0554, tangentMode=RCTM_User, time=1, value=0}
        ]}}})
    Scale RGB=&VectorFromFloat(Value=&FloatFromCurve(
      FloatCurve=DataInterface() {propertyValues {Curve {keys [
          <at 0 1 RCIM_Cubic>
          {arriveTangent=-1.3601, interpMode=RCIM_Cubic, leaveTangent=-1.3601, time=0.2, value=0.95}
          {arriveTangent=-0.1237, interpMode=RCIM_Cubic, leaveTangent=-0.1237, tangentMode=RCTM_User, time=0.3617, value=0.508}
          <at 1 0 RCIM_Cubic>
         ]}}}
      Scale Curve=&Multiply_Float(A=2, B=1)
     ))
    ScaleA=true
    ScaleRGB=true
   )
  ]
  rendererClasses [&NiagaraMeshRendererProperties]
  renderers [{rendererClass=&NiagaraMeshRendererProperties, rendererIndex=0}]
 }
 FireTrailMesh001 {
  emitterUpdateScript [
   &EmitterState(Inactive Response=::NewEnumerator1, Life Cycle Mode=::NewEnumerator1, Loop Behavior=::NewEnumerator1, Loop Duration=1, UseLoopDelay=false)
   &SpawnBurst_Instantaneous(Spawn Count=2, Spawn Time=0.4, Use Spawn Probability=false)
  ]
  particleSpawnScript [
   &PartitionParticles(Number Of Partitions=4)
   &InitializeParticle(
    Lifetime=&Multiply_Float(
     A=0.7
     B=&SelectFloatFromArray(
      Direct Array Index=$Particles.PartitionParticles.Partition
      Float Selection Array=DataInterface() {propertyValues {FloatData [1,0.9,0.8,0.7]}}
     )
    )
    Mesh Scale Mode=::NewEnumerator3
    Mesh Uniform Scale=&Multiply_Float(
     A=&SelectFloatFromArray(
      Direct Array Index=$Particles.PartitionParticles.Partition
      Float Selection Array=DataInterface() {propertyValues {FloatData [1,0.68,0.45,0.26]}}
     )
     B=0.42
    )
    UsePositionOffset=false
    Write Material Random=true
   ) {module=InitializeParticle001}
   &InitialMeshOrientation(
    Mesh Orientation Mode=::NewEnumerator2
    Rotation=&MakeVector(
     X=&Add_Float(
      A=&SelectFloatFromArray(
       Direct Array Index=$Particles.PartitionParticles.Partition
       Float Selection Array=DataInterface() {propertyValues {FloatData [0,-0.03,-0.07,-0.09]}}
      )
      B=0.6
     )
     Y=0
     Z=0
    )
    Use Rotation Vector=true
   )
   &SystemLocation(MaskSpawnGroup=false)
  ]
  particleUpdateScript [
   &ParticleState()
   &ScaleMeshSize(Scale Factor=<1 1.5 1>)
   &SolveForcesAndVelocity(Clamp Velocity=false, Limit Acceleration=false)
   &DynamicMaterialParameters(
    Index 0 Param 1=&FloatFromCurve(
     FloatCurve=DataInterface() {propertyValues {Curve {keys [{arriveTangent=3.1068, interpMode=RCIM_Cubic, leaveTangent=3.1068, tangentMode=RCTM_User, time=0, value=0}, <at 1 1 RCIM_Cubic>]}}}
     Scale Curve=0.7
    )
    Index 0 Param 2=&Multiply_Float(
     A=0.7
     B=&SelectFloatFromArray(Direct Array Index=$Particles.PartitionParticles.Partition, Float Selection Array=DataInterface() {propertyValues {FloatData [1,1,1,1]}})
    )
    Index 0 Param 3=1
    Index 0 Param 4=2
    Module.Index0Float0Write=true
    Module.Index0Float1Write=true
    Module.Index0Float2Write=true
    Module.Index0Float3Write=true
   )
   &ScaleColor(
    Scale Alpha=&FloatFromCurve(FloatCurve=DataInterface() {propertyValues {Curve {keys [
         {interpMode=RCIM_Cubic, tangentMode=RCTM_User, time=0, value=0}
         {arriveTangent=0.25, interpMode=RCIM_Cubic, leaveTangent=0.25, time=0.4, value=0.95}
         {arriveTangent=-1.5833, interpMode=RCIM_Cubic, leaveTangent=-1.5833, time=0.6, value=0.15}
         {arriveTangent=-0.0554, interpMode=RCIM_Cubic, leaveTangent=-0.0554, tangentMode=RCTM_User, time=1, value=0}
        ]}}})
    Scale RGB=&VectorFromFloat(Value=&FloatFromCurve(
      FloatCurve=DataInterface() {propertyValues {Curve {keys [
          <at 0 1 RCIM_Cubic>
          {arriveTangent=-1.3601, interpMode=RCIM_Cubic, leaveTangent=-1.3601, time=0.2, value=0.95}
          {arriveTangent=-0.1237, interpMode=RCIM_Cubic, leaveTangent=-0.1237, tangentMode=RCTM_User, time=0.3617, value=0.508}
          <at 1 0 RCIM_Cubic>
         ]}}}
      Scale Curve=&Multiply_Float(A=1, B=1)
     ))
    ScaleA=true
    ScaleRGB=true
   )
  ]
  rendererClasses [&NiagaraMeshRendererProperties]
  renderers [{rendererClass=&NiagaraMeshRendererProperties, rendererIndex=0}]
 }
 Fire_Sprite {
  emitterUpdateScript [
   &EmitterState()
   &SpawnRate(SpawnRate=10, Use Spawn Probability=false)
  ]
  particleSpawnScript [
   &InitializeParticle(
    Lifetime Max=1
    Lifetime Min=0.6
    Lifetime Mode=::NewEnumerator1
    Sprite Rotation Mode=::NewEnumerator2
    Sprite Size Mode=::NewEnumerator1
    Uniform Sprite Size Max=30
    Uniform Sprite Size Min=15
    UsePositionOffset=false
    Write Material Random=true
   )
   &DynamicMaterialParameters(
    Index 0 Param 2=&RandomRangeFloat(Minimum=0, Maximum=4)
    Index 0 Param 4=2.5
    Module.Param0WriteEnabled=false
    Module.Param1WriteEnabled=true
    Module.Param2WriteEnabled=false
    Module.Param3WriteEnabled=true
   ) {module=DynamicMaterialParameters001}
   &AddVelocity(
    Cone Angle=8
    Module.UseConeFalloff=false
    Module.UseVelDistribution=false
    Override Seed=false
    Velocity Mode=::NewEnumerator2
    Velocity Speed=&Multiply_Float(A=&RandomRangeFloat(Minimum=0.8, Maximum=1), B=1200)
   ) {enabled=false, module=AddVelocity001}
   &CameraOffset(Camera Offset Amount=&RandomRangeFloat(Minimum=5, Maximum=25))
  ]
  particleUpdateScript [
   &ParticleState()
   &AccelerationForce(Acceleration=&UniformRangedVector(Minimum=<0 0 15>, Maximum=<0 0 25>))
   &Drag(
    Drag=&FloatFromCurve(FloatCurve=DataInterface() {propertyValues {Curve {keys [<at 0 0.075 RCIM_Cubic>,<at 1 1.075 RCIM_Cubic>]}}}, Scale Curve=4)
    Use Linear Drag=true
    Use Rotational Drag=false
   )
   &ScaleColor(
    Linear Color Curve=DataInterface() {propertyValues {
      AlphaCurve {keys [{arriveTangent=-0.1596, interpMode=RCIM_Cubic, leaveTangent=-0.1596, tangentMode=RCTM_User, time=0.8987, value=1}, <at 0.982 0 RCIM_Cubic>]}
      BlueCurve {keys [
        <at 0.0505 3.4995 RCIM_Cubic>
        {arriveTangent=-3.8612, interpMode=RCIM_Cubic, leaveTangent=-3.8612, time=0.3055, value=0.902}
        {arriveTangent=-0.6273, interpMode=RCIM_Cubic, leaveTangent=-0.6273, time=0.7879, value=0.6522}
        <at 0.9823 0.4774>
       ]}
      GreenCurve {keys [
        <at 0.0505 4.0716 RCIM_Cubic>
        {arriveTangent=-4.5848, interpMode=RCIM_Cubic, leaveTangent=-4.5848, time=0.3055, value=1.3743}
        {arriveTangent=-1.1628, interpMode=RCIM_Cubic, leaveTangent=-1.1628, time=0.7879, value=0.6908}
        <at 0.9823 0.5872>
       ]}
      RedCurve {keys [
        <at 0.0505 5 RCIM_Cubic>
        {arriveTangent=-5.4244, interpMode=RCIM_Cubic, leaveTangent=-5.4244, time=0.3055, value=3}
        {arriveTangent=-2.955, interpMode=RCIM_Cubic, leaveTangent=-2.955, time=0.7879, value=1}
        <at 0.9823 1>
       ]}
     }}
    Scale Mode=::NewEnumerator2
   ) {module=ScaleColor001}
   &ScaleColor(
    Scale Alpha=&FloatFromCurve(FloatCurve=DataInterface() {propertyValues {Curve {keys [
         {arriveTangent=6.7724, interpMode=RCIM_Cubic, leaveTangent=6.7724, tangentMode=RCTM_User, time=0, value=0}
         {interpMode=RCIM_Cubic, tangentMode=RCTM_User, time=0.5, value=1}
         {arriveTangent=-3.0347, interpMode=RCIM_Cubic, leaveTangent=-3.0347, tangentMode=RCTM_User, time=1, value=0}
        ]}}})
    ScaleA=true
    ScaleRGB=true
   )
   &ScaleColor(
    Scale Alpha=&FloatFromCurve(
     FloatCurve=DataInterface() {propertyValues {Curve {keys [
         {arriveTangent=-0.0766, interpMode=RCIM_Cubic, leaveTangent=-0.0766, tangentMode=RCTM_User, time=0, value=0}
         {arriveTangent=1, interpMode=RCIM_Cubic, leaveTangent=1, time=0.4, value=0.5}
         {interpMode=RCIM_Cubic, tangentMode=RCTM_User, time=1, value=1}
        ]}}}
     Scale Curve=&Lerp_Float(A=0, B=1, Alpha=$System.ScaleUpCurve)
    )
    ScaleA=true
    ScaleRGB=true
   ) {enabled=false, module=ScaleColor002}
   &SolveForcesAndVelocity(Clamp Velocity=false, Limit Acceleration=false)
   &ScaleSpriteSize(Scale Factor=&Multiply_Vector2D_ByFloat(
     Float=&FloatFromCurve(
      FloatCurve=DataInterface() {propertyValues {Curve {keys [
          {arriveTangent=1.4708, interpMode=RCIM_Cubic, leaveTangent=1.4708, tangentMode=RCTM_User, time=0, value=0.875}
          {arriveTangent=1.125, interpMode=RCIM_Cubic, leaveTangent=1.125, time=0.6, value=1.625}
          {interpMode=RCIM_Cubic, tangentMode=RCTM_User, time=1, value=2}
         ]}}}
      Scale Curve=&Lerp_Float(A=0, B=1, Alpha=$System.ScaleUpCurve)
     )
     Vector2D=<1 1>
    )) {module=ScaleSpriteSize001}
  ]
  rendererClasses [&NiagaraSpriteRendererProperties]
  renderers [{rendererClass=&NiagaraSpriteRendererProperties}]
 }
 Fire_Sprite001 {
  emitterUpdateScript [
   &EmitterState(Inactive Response=::NewEnumerator1, Life Cycle Mode=::NewEnumerator1, Loop Behavior=::NewEnumerator1, Loop Duration=1, UseLoopDelay=false)
   &SpawnBurst_Instantaneous(Spawn Count=4, Spawn Time=0.5, Use Spawn Probability=false, UseLoopCountLimit=false)
  ]
  particleSpawnScript [
   &InitializeParticle(
    Lifetime Max=0.6
    Lifetime Min=0.5
    Lifetime Mode=::NewEnumerator1
    Sprite Rotation Mode=::NewEnumerator2
    Sprite Size Mode=::NewEnumerator1
    Uniform Sprite Size Max=30
    Uniform Sprite Size Min=15
    UsePositionOffset=false
    Write Material Random=true
   )
   &DynamicMaterialParameters(
    Index 0 Param 2=&RandomRangeFloat(Minimum=0, Maximum=4)
    Index 0 Param 4=2.5
    Module.Param0WriteEnabled=false
    Module.Param1WriteEnabled=true
    Module.Param2WriteEnabled=false
    Module.Param3WriteEnabled=true
   ) {module=DynamicMaterialParameters001}
   &AddVelocity(
    Cone Angle=8
    Module.UseConeFalloff=false
    Module.UseVelDistribution=false
    Override Seed=false
    Velocity Mode=::NewEnumerator2
    Velocity Speed=&Multiply_Float(A=&RandomRangeFloat(Minimum=0.8, Maximum=1), B=1200)
   ) {enabled=false, module=AddVelocity001}
   &CameraOffset(Camera Offset Amount=&RandomRangeFloat(Minimum=5, Maximum=25))
  ]
  particleUpdateScript [
   &ParticleState()
   &AccelerationForce(Acceleration=&UniformRangedVector(Minimum=<0 0 15>, Maximum=<0 0 25>))
   &Drag(
    Drag=&FloatFromCurve(FloatCurve=DataInterface() {propertyValues {Curve {keys [<at 0 0.075 RCIM_Cubic>,<at 1 1.075 RCIM_Cubic>]}}}, Scale Curve=4)
    Use Linear Drag=true
    Use Rotational Drag=false
   )
   &ScaleColor(
    Linear Color Curve=DataInterface() {propertyValues {
      AlphaCurve {keys [{arriveTangent=-0.1596, interpMode=RCIM_Cubic, leaveTangent=-0.1596, tangentMode=RCTM_User, time=0.8987, value=1}, <at 0.982 0 RCIM_Cubic>]}
      BlueCurve {keys [
        <at 0.0505 3.4995 RCIM_Cubic>
        {arriveTangent=-3.8612, interpMode=RCIM_Cubic, leaveTangent=-3.8612, time=0.3055, value=0.902}
        {arriveTangent=-0.6273, interpMode=RCIM_Cubic, leaveTangent=-0.6273, time=0.7879, value=0.6522}
        <at 0.9823 0.4774>
       ]}
      GreenCurve {keys [
        <at 0.0505 4.0716 RCIM_Cubic>
        {arriveTangent=-4.5848, interpMode=RCIM_Cubic, leaveTangent=-4.5848, time=0.3055, value=1.3743}
        {arriveTangent=-1.1628, interpMode=RCIM_Cubic, leaveTangent=-1.1628, time=0.7879, value=0.6908}
        <at 0.9823 0.5872>
       ]}
      RedCurve {keys [
        <at 0.0505 5 RCIM_Cubic>
        {arriveTangent=-5.4244, interpMode=RCIM_Cubic, leaveTangent=-5.4244, time=0.3055, value=3}
        {arriveTangent=-2.955, interpMode=RCIM_Cubic, leaveTangent=-2.955, time=0.7879, value=1}
        <at 0.9823 1>
       ]}
     }}
    Scale Mode=::NewEnumerator2
   ) {module=ScaleColor001}
   &ScaleColor(
    Scale Alpha=&FloatFromCurve(FloatCurve=DataInterface() {propertyValues {Curve {keys [
         {arriveTangent=6.7724, interpMode=RCIM_Cubic, leaveTangent=6.7724, tangentMode=RCTM_User, time=0, value=0}
         {interpMode=RCIM_Cubic, tangentMode=RCTM_User, time=0.5, value=1}
         {arriveTangent=-3.0347, interpMode=RCIM_Cubic, leaveTangent=-3.0347, tangentMode=RCTM_User, time=1, value=0}
        ]}}})
    ScaleA=true
    ScaleRGB=true
   )
   &SolveForcesAndVelocity(Clamp Velocity=false, Limit Acceleration=false)
   &ScaleSpriteSize(Scale Factor=&Multiply_Vector2D_ByFloat(
     Float=&FloatFromCurve(
      FloatCurve=DataInterface() {propertyValues {Curve {keys [
          {arriveTangent=1.4708, interpMode=RCIM_Cubic, leaveTangent=1.4708, tangentMode=RCTM_User, time=0, value=0.875}
          {arriveTangent=1.125, interpMode=RCIM_Cubic, leaveTangent=1.125, time=0.6, value=1.625}
          {interpMode=RCIM_Cubic, tangentMode=RCTM_User, time=1, value=2}
         ]}}}
      Scale Curve=&Lerp_Float(A=0.5, B=1, Alpha=$System.ScaleUpCurve)
     )
     Vector2D=&Vector2DFromFloat(Value=2)
    )) {module=ScaleSpriteSize001}
  ]
  rendererClasses [&NiagaraSpriteRendererProperties]
  renderers [{rendererClass=&NiagaraSpriteRendererProperties}]
 }
 Flash_Flare {
  emitterUpdateScript [
   &EmitterState(
    Inactive Response=::NewEnumerator1
    Life Cycle Mode=::NewEnumerator1
    Loop Behavior=::NewEnumerator1
    Loop Duration Mode=::NewEnumerator1
    UseLoopDelay=false
   )
   &SpawnBurst_Instantaneous(Spawn Count=1, Use Spawn Probability=false)
  ]
  particleSpawnScript [
   &InitializeParticle(
    Color=<rgba 0.6042 0.1287 0 1>
    Lifetime=0.2
    Sprite Size Mode=::NewEnumerator3
    Uniform Sprite Size=250
    UsePositionOffset=false
    Write Material Random=true
   )
   &DynamicMaterialParameters(
    Index 0 Param 1=0.2
    Index 0 Param 2=1
    Module.Param0WriteEnabled=true
    Module.Param1WriteEnabled=true
    Module.Param2WriteEnabled=true
    Module.Param3WriteEnabled=true
   )
   &CameraOffset(Camera Offset Amount=10)
  ]
  particleUpdateScript [
   &ParticleState(Kill Particles When Lifetime Has Elapsed=false, Let Infinitely Lived Particles Die When Emitter Deactivates=true)
   &ScaleSpriteSize(
    Scale Sprite Size Mode=::NewEnumerator0
    Uniform Scale Factor=&Lerp_Float(
     A=0.6
     B=1
     Alpha=&FloatFromCurve(
      FloatCurve=DataInterface() {propertyValues {Curve {keys [{tangentMode=RCTM_User, time=0.7, value=0}, {tangentMode=RCTM_User, time=1.5, value=1}]}}}
      CurveIndex=$System.Age
     )
    )
   ) {module=ScaleSpriteSize001}
   &ScaleSpriteSize(Scale Factor=<1.2 0.5>, Scale Sprite Size Mode=::NewEnumerator2)
   &ScaleSpriteSize(Scale Sprite Size Mode=::NewEnumerator0, Uniform Scale Factor=&Lerp_Float(A=0.5, B=1, Alpha=$System.ScaleUpCurve)) {module=ScaleSpriteSize002}
   &ScaleColor(
    Scale Alpha=&FloatFromCurve(
     FloatCurve=DataInterface() {propertyValues {Curve {keys [{tangentMode=RCTM_User, time=0, value=0}, <at 0.2 1>, {tangentMode=RCTM_User, time=1, value=1}]}}}
     CurveIndex=$System.Age
     Scale Curve=0.25
    )
    Scale RGB=&VectorFromFloat(Value=1)
    ScaleA=true
    ScaleRGB=true
   )
   &FN_Light_Attributes(
    Color=$Particles.Color
    Create Color Attribute=true
    Scale Color=&FloatFromCurve(
     FloatCurve=DataInterface() {propertyValues {Curve {keys [{tangentMode=RCTM_User, time=0, value=0}, {tangentMode=RCTM_User, time=0.1, value=1}]}}}
     CurveIndex=$System.Age
    )
   )
   /SetVariables(Particles.LightRadius=&Multiply_Float(A=300, B=&Lerp_Float(A=0.5, B=1, Alpha=$System.ScaleUpCurve))) {isSetParameters=true}
   &ScaleColor(
    Scale RGB=&VectorFromFloat(Value=&Lerp_Float(
      A=1
      B=&RandomRangeFloat(Minimum=0.5, Maximum=3, Evaluation Type=::NewEnumerator1)
      Alpha=&FloatFromCurve(
       FloatCurve=DataInterface() {propertyValues {Curve {keys [{tangentMode=RCTM_User, time=1.3, value=0}, {tangentMode=RCTM_User, time=1.5, value=1}]}}}
       CurveIndex=$System.Age
      )
     ))
    ScaleA=true
    ScaleRGB=true
   ) {module=ScaleColor001}
  ]
  rendererClasses [&NiagaraSpriteRendererProperties, &NiagaraLightRendererProperties]
  renderers [
   {rendererClass=&NiagaraSpriteRendererProperties}
   {rendererClass=&NiagaraLightRendererProperties, rendererIndex=1}
   {rendererClass=&NiagaraSpriteRendererProperties, rendererIndex=2}
  ]
 }
}
summary {
 systemName=NS_NS_SoapPocket_Baking
 emitters [
  {emitterName=Flash_Flare, rendererClasses [&NiagaraSpriteRendererProperties, &NiagaraLightRendererProperties]}
  {emitterName=Fire_Sprite, rendererClasses [&NiagaraSpriteRendererProperties]}
  {emitterName=FireTrailMesh, rendererClasses [&NiagaraMeshRendererProperties]}
  {emitterName=Fire_Sprite001, rendererClasses [&NiagaraSpriteRendererProperties]}
  {emitterName=FireTrailMesh001, rendererClasses [&NiagaraMeshRendererProperties]}
 ]
 userVariables [
  User.EnableRopeTightenFX {type=NiagaraBool}
  User.RopeLength {type=NiagaraFloat, defaultValue=500, description=""}
 ]
}
